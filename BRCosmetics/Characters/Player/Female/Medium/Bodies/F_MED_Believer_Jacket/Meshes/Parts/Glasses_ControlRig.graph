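# Control Rig for the Believer Jacket female character's glasses that nudges dyn_glasses as the cheek rises.
# The left inner cheek's height relative to faceAttach is remapped to a clamped 0–1 weight.
# That weight drives an additive local bone modify applying a small offset and yaw.
# A right-cheek remap feeds a Max node with the left, but that result is unconnected, so only the left cheek has effect.
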ControlRigBlueprint /BRCosmetics/Characters/Player/Female/Medium/Bodies/F_MED_Believer_Jacket/Meshes/Parts/Glasses_ControlRig
graphs [
 RigVMFunctionLibrary {links [], nodes []}
 RigVMModel {nodes [
   n1 = &RigUnit_BeginExecution()
   n2 = &RigVMFunction_MathFloatRemap(
    SourceMaximum=-1.28
    SourceMinimum=-1.4
    TargetMinimum=0
    bClamp=true
    Value=&RigUnit_GetRelativeBoneTransform(Bone=L_cheek_inner, Space=faceAttach).Transform.Translation.Z
   )
   /Script/ControlRig.RigUnit_ModifyBoneTransforms(
    BoneToModify=((Bone="dyn_glasses",Transform=(Rotation=(X=0,Y=0,Z=0,W=1),Translation=(X=0,Y=0.5,Z=0),Scale3D=(X=1,Y=1,Z=1))))
    Mode=AdditiveLocal
    WeightMaximum=1
    WeightMinimum=0
    ExecutePin=@n1.ExecutePin
    Weight=@n2.Result
    BoneToModify.0.Transform.Rotation=/Script/RigVM.RigVMFunction_MathQuaternionFromRotator(Rotator=(Pitch=0,Yaw=-3,Roll=0)) {is=struct RigVMUnitNode}.Result
   ) {is=struct RigVMUnitNode}
   /Script/RigVM.RigVMFunction_MathFloatMax(
    A=@n2.Result
    B=&RigVMFunction_MathFloatRemap(
     SourceMaximum=-1.28
     SourceMinimum=-1.4
     TargetMinimum=0
     bClamp=true
     Value=&RigUnit_GetRelativeBoneTransform(Bone=R_cheek_inner, Space=faceAttach).Transform.Translation.Z
    ).Result
   ) {is=struct RigVMUnitNode, template="Maximum::Execute(in A,in B,out Result)"}
  ]}
]
hierarchy [
 root {loc=<0 0 0>, rot=<0 0 0 1>}
 pelvis {parent=root, loc=<0 -0.5301 98.0571>, rot=<0.0127 0.707 -0.0127 -0.707>}
 spine_01 {parent=pelvis, loc=<6.1208 -0.9328 0>, rot=<0 0 0.034 -0.9994>}
 spine_02 {parent=spine_01, loc=<6.5996 -0.2095 0>, rot=<0 0 0.0238 -0.9997>}
 spine_03 {parent=spine_02, loc=<6.4914 -0.1237 0>, rot=<0 0 0.0149 -0.9999>}
 spine_04 {parent=spine_03, loc=<6.9598 0.0794 0>, rot=<0 0 -0.0132 -0.9999>}
 spine_05 {parent=spine_04, loc=<6.9987 0 0>, rot=<0 0 -0.0015 -1>}
 neck_01 {parent=spine_05, loc=<11.3138 0.1529 0>, rot=<0 0 -0.2003 -0.9797>}
 neck_02 {parent=neck_01, loc=<4.5627 0 0>, rot=<0 0 0.0504 -0.9987>}
 head {parent=neck_02, loc=<5.3914 0 0>, rot=<0 0 0.115 -0.9934>}
 hair_side_strand_r {parent=head, loc=<7.4285 4.6764 6.364>, rot=<0.983 0.0991 -0.0077 0.1545>}
 hair_side_strandB_a_r {parent=hair_side_strand_r, loc=<-5.5715 0.6793 -0.1268>, rot=<0.0066 -0.023 -0.2754 -0.961>}
 hair_side_strandB_b_r {parent=hair_side_strandB_a_r, loc=<-2.9652 -0.1232 0.375>, rot=<0.0177 -0.0835 -0.2069 -0.9746>}
 hair_side_strandA_a_r {parent=hair_side_strand_r, loc=<-5.6702 -2.1744 -0.1351>, rot=<0.0243 -0.0687 -0.3327 -0.9402>}
 hair_side_strandA_a_l {parent=head, loc=<4.9045 6.0276 -6.1565>, rot=<0.0987 -0.9877 -0.121 -0.0111>}
 hair_side_strandA_b_l {parent=hair_side_strandA_a_l, loc=<3.8601 0.6414 -0.1842>, rot=<0.0492 -0.1436 -0.32 -0.9352>}
 hair_side_strandB_a_l {parent=head, loc=<3.6824 3.1984 -7.0249>, rot=<0.0568 -0.991 -0.1214 -0.006>}
 hair_side_strandB_b_l {parent=hair_side_strandB_a_l, loc=<3.1035 -0.0362 -0.4169>, rot=<0.0366 -0.1527 -0.23 -0.9604>}
 hair_side_l {parent=head, loc=<5.811 -5.3596 -5.8742>, rot=<0.0037 -0.8866 0.4625 -0.0019>}
 hair_bottomA_l {parent=head, loc=<-1.3898 -1.9778 -6.0192>, rot=<-0.3508 -0.9025 -0.1674 0.1856>}
 hair_side_r {parent=head, loc=<5.8087 -5.0467 6.3228>, rot=<0.8866 0.0037 -0.0019 -0.4625>}
 hair_bottomA_r {parent=head, loc=<-1.3899 -1.9778 6.0192>, rot=<0.9025 -0.3508 0.1856 0.1674>}
 hair_fr_l {parent=head, loc=<8.7613 9.599 -1.671>, rot=<0.0041 -0.9845 0.1753 -0.0007>}
 hair_fr_r {parent=head, loc=<9.2099 9.3473 2.193>, rot=<0.0041 -0.9926 -0.1215 0.0005>}
 hair_bk {parent=head, loc=<6.1274 -7.4073 0>, rot=<0.0029 -0.7071 0.7071 -0.0029>}
 dyn_glasses {parent=head, loc=<4.3233 0 0>, rot=<0 0 0 -1>}
 faceAttach {parent=head, loc=<0.6201 0.9296 0>, rot=<0.0082 -0.7071 -0.0082 -0.7071>, scale=<0.89 0.89 0.89>}
 C_brow_mid {parent=faceAttach, loc=<-0.009 8.997 3.9416>, rot=<0.0005 0.0142 0.0083 -0.9999>, scale=<0.4914 0.8 0.6189>}
 L_brow_mid {parent=faceAttach, loc=<3.3266 8.2699 4.963>, rot=<0.0087 0.028 0.2634 -0.9642>}
 L_brow_outer {parent=faceAttach, loc=<5.9352 6.2161 4.7772>, rot=<0.2365 0.4849 0.387 -0.7478>}
 L_cheek_inner {parent=faceAttach, loc=<4.4419 7.0682 -1.5025>, rot=<-0.0949 -0.3633 0.3754 -0.8474>}
 C_lip_upper_mid {parent=faceAttach, loc=<0.0048 8.7671 -3.5476>, rot=<0.0776 0.0016 -0.0009 -0.997>}
 L_lip_upper_outer {parent=faceAttach, loc=<1.2083 8.0653 -3.6019>, rot=<0.0304 -0.077 0.2029 -0.9757>}
 L_lip_corner {parent=faceAttach, loc=<2.3189 7.6996 -3.8633>, rot=<0.0006 0.0086 0.2452 -0.9694>}
 C_nose_bridge {parent=faceAttach, loc=<-0.0009 9.7555 0.6469>, rot=<-0.2218 0.0044 0.0013 -0.9751>}
 R_brow_mid {parent=faceAttach, loc=<-3.5035 8.1853 5.0824>, rot=<0.0237 -0.0447 -0.2511 -0.9666>}
 R_brow_outer {parent=faceAttach, loc=<-5.9043 6.2631 4.6551>, rot=<0.2365 -0.4849 -0.387 -0.7478>}
 R_cheek_inner {parent=faceAttach, loc=<-4.4419 7.0682 -1.5025>, rot=<-0.0949 0.3633 -0.3754 -0.8474>}
 R_lip_corner {parent=faceAttach, loc=<-2.3189 7.6996 -3.8633>, rot=<0.0006 -0.0086 -0.2452 -0.9694>}
 C_jaw {parent=faceAttach, loc=<0.2255 0.4069 -2.6176>, rot=<0 0 0 1>}
 R_eye {parent=faceAttach, loc=<-3.4594 5.7481 3.1608>, rot=<0 0 0.9998 -0.0187>}
 L_eye {parent=faceAttach, loc=<3.4592 5.7481 3.1608>, rot=<0 0 0.9998 0.0187>}
 teeth_upper {parent=faceAttach, loc=<0.0016 6.6099 -2.7708>, rot=<0 0 0 1>}
 R_lip_upper_outer {parent=faceAttach, loc=<-1.2083 8.0653 -3.6019>, rot=<0.0304 0.077 -0.2029 -0.9757>}
]
