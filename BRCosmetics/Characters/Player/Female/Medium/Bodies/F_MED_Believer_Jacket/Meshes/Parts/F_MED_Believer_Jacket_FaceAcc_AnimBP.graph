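# Animation blueprint for the Believer Jacket female character's face accessory part, adding secondary hair motion and a glasses rig.
# It copies the attached parent's pose, then runs chained spring-driven AnimDynamics nodes on front, side, back, strand and bottom hair bones.
# A glasses Control Rig follows, weighted by an update-computed alpha that enables it only at LOD0 outside the front end.
# The isFrontEnd variable is inverted (zero in front end), and the layered-blend node is unconnected.
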AnimBlueprint /BRCosmetics/Characters/Player/Female/Medium/Bodies/F_MED_Believer_Jacket/Meshes/Parts/F_MED_Believer_Jacket_FaceAcc_AnimBP
target_skeleton=&/F_MED_Believer_Jacket_FaceAcc_Skeleton
graphs [
 AnimGraph {nodes [
   Layered blend per bone(
    BasePose=Copy Pose From Mesh() {bCopyCurves=true, bUseAttachedParent=true}
    BlendPoses_0=Copy Pose From Mesh(SourceMeshComponent=Get Body_Mesh())
   )
   n1 = Anim Dynamics - Bone: hair_fr_l(ComponentPose=Local To Component(LocalPose=Copy Pose From Mesh() {bUseAttachedParent=true}), Alpha=1, ExternalForce="0,0,0") {
    angularDampingOverride=1
    angularSpringConstant=1500
    bAngularSpring=true
    bGravityOverrideInSimSpace=true
    bOverrideAngularDamping=true
    gravityScale=0.35
    simulationSpace=RootRelative
    alphaBoolBlend {blendOption=Linear}
    boundBone {boneName=hair_fr_l}
    componentLinearAccScale=<0.35 0.35 0.35>
    componentLinearVelScale=<0.35 0.35 0.35>
    physicsBodyDefinitions [{
      sphereCollisionRadius=0
      boundBone {boneName=hair_fr_l}
      boxExtents=<10 20 20>
      constraintSetup {angularLimitsMax=<0 5 30>, angularLimitsMin=<0 -5 0>, angularTarget=<0 0 0>}
      localJointOffset=<5 0 0>
     }]
   }
   n2 = Anim Dynamics - Bone: hair_fr_r(ComponentPose=@n1, Alpha=1, ExternalForce="0,0,0") {
    angularDampingOverride=1
    angularSpringConstant=1500
    bAngularSpring=true
    bGravityOverrideInSimSpace=true
    bOverrideAngularDamping=true
    gravityScale=0.35
    simulationSpace=RootRelative
    alphaBoolBlend {blendOption=Linear}
    boundBone {boneName=hair_fr_r}
    physicsBodyDefinitions [{
      sphereCollisionRadius=0
      boundBone {boneName=hair_fr_r}
      boxExtents=<10 20 20>
      constraintSetup {angularLimitsMax=<0 5 30>, angularLimitsMin=<0 -5 0>, angularTarget=<0 0 0>}
      localJointOffset=<5 0 0>
     }]
   }
   n3 = Anim Dynamics - Bone: hair_side_l(ComponentPose=@n2, Alpha=0.6, ExternalForce="0,0,0") {
    angularDampingOverride=1
    angularSpringConstant=1500
    bAngularSpring=true
    bGravityOverrideInSimSpace=true
    bOverrideAngularDamping=true
    gravityScale=0.35
    simulationSpace=RootRelative
    alphaBoolBlend {blendOption=Linear}
    boundBone {boneName=hair_side_l}
    componentLinearAccScale=<0.35 0.35 0.3>
    componentLinearVelScale=<0.35 0.35 0.3>
    physicsBodyDefinitions [{
      sphereCollisionRadius=0
      boundBone {boneName=hair_side_l}
      boxExtents=<7 12 12>
      constraintSetup {angularLimitsMax=<5 25 15>, angularLimitsMin=<-5 -35 -15>, angularTarget=<0 0 0>}
      localJointOffset=<5 0 0>
     }]
   }
   n4 = Anim Dynamics - Bone: hair_side_r(ComponentPose=@n3, Alpha=0.6, ExternalForce="0,0,0") {
    angularDampingOverride=1
    angularSpringConstant=1500
    bAngularSpring=true
    bGravityOverrideInSimSpace=true
    bOverrideAngularDamping=true
    gravityScale=0.35
    simulationSpace=RootRelative
    alphaBoolBlend {blendOption=Linear}
    boundBone {boneName=hair_side_r}
    componentLinearAccScale=<0.35 0.35 0.3>
    componentLinearVelScale=<0.35 0.35 0.3>
    physicsBodyDefinitions [{
      sphereCollisionRadius=0
      boundBone {boneName=hair_side_r}
      boxExtents=<7 12 12>
      constraintSetup {angularLimitsMax=<5 25 15>, angularLimitsMin=<-5 -35 -15>, angularTarget=<0 0 0>}
      localJointOffset=<-5 0 0>
     }]
   }
   n5 = Anim Dynamics - Bone: hair_bk(ComponentPose=@n4, Alpha=0.7, ExternalForce="0,0,0") {
    angularDampingOverride=1
    angularSpringConstant=1500
    bAngularSpring=true
    bGravityOverrideInSimSpace=true
    bOverrideAngularDamping=true
    gravityScale=0.35
    simulationSpace=RootRelative
    alphaBoolBlend {blendOption=Linear}
    boundBone {boneName=hair_bk}
    componentLinearAccScale=<0.35 0.35 0.3>
    componentLinearVelScale=<0.35 0.35 0.3>
    physicsBodyDefinitions [{
      sphereCollisionRadius=0
      boundBone {boneName=hair_bk}
      boxExtents=<7 12 12>
      constraintSetup {angularLimitsMax=<5 25 15>, angularLimitsMin=<-5 -35 -15>, angularTarget=<0 0 0>}
      localJointOffset=<5 0 0>
     }]
   }
   n6 = Anim Dynamics - Bone: hair_side_strandA_a_l(ComponentPose=@n5, Alpha=0.8, ExternalForce="0,0,0") {
    angularDampingOverride=1
    angularSpringConstant=1500
    bAngularSpring=true
    bGravityOverrideInSimSpace=true
    bOverrideAngularDamping=true
    gravityScale=0.35
    simulationSpace=RootRelative
    alphaBoolBlend {blendOption=Linear}
    boundBone {boneName=hair_side_strandA_a_l}
    componentLinearAccScale=<0.3 0.3 0.3>
    componentLinearVelScale=<0.3 0.3 0.3>
    physicsBodyDefinitions [{
      sphereCollisionRadius=0
      boundBone {boneName=hair_side_strandA_a_l}
      boxExtents=<10 15 15>
      constraintSetup {angularLimitsMax=<5 0 20>, angularLimitsMin=<-5 -20 -35>, angularTarget=<0 0 0>, linearAxesMax=<0.3 0.3 0.3>, linearAxesMin=<0.3 0.3 0.3>}
      localJointOffset=<5 0 0>
     }]
   }
   n7 = Anim Dynamics - Bone: hair_side_strandA_b_l(ComponentPose=@n6, Alpha=0.8, ExternalForce="0,0,0") {
    angularDampingOverride=1
    angularSpringConstant=1500
    bAngularSpring=true
    bGravityOverrideInSimSpace=true
    bOverrideAngularDamping=true
    gravityScale=0.35
    simulationSpace=RootRelative
    alphaBoolBlend {blendOption=Linear}
    boundBone {boneName=hair_side_strandA_b_l}
    componentLinearAccScale=<0.3 0.3 0.3>
    componentLinearVelScale=<0.3 0.3 0.3>
    physicsBodyDefinitions [{
      sphereCollisionRadius=0
      boundBone {boneName=hair_side_strandA_b_l}
      boxExtents=<10 15 15>
      constraintSetup {angularLimitsMax=<5 0 20>, angularLimitsMin=<-5 -20 -35>, angularTarget=<0 0 0>, linearAxesMax=<0.3 0.3 0.3>, linearAxesMin=<0.3 0.3 0.3>}
      localJointOffset=<5 0 0>
     }]
   }
   n8 = Anim Dynamics - Bone: hair_side_strandB_a_l(ComponentPose=@n7, Alpha=1, ExternalForce="0,0,0") {
    angularDampingOverride=1
    angularSpringConstant=1500
    bAngularSpring=true
    bGravityOverrideInSimSpace=true
    bOverrideAngularDamping=true
    gravityScale=0.35
    simulationSpace=RootRelative
    alphaBoolBlend {blendOption=Linear}
    boundBone {boneName=hair_side_strandB_a_l}
    componentLinearAccScale=<0.3 0.3 0.3>
    componentLinearVelScale=<0.3 0.3 0.3>
    physicsBodyDefinitions [{
      sphereCollisionRadius=0
      boundBone {boneName=hair_side_strandB_a_l}
      boxExtents=<10 15 15>
      constraintSetup {angularLimitsMax=<5 0 10>, angularLimitsMin=<-5 -20 -25>, angularTarget=<0 0 0>, linearAxesMax=<0.2 0.2 0.2>, linearAxesMin=<0.2 0.2 0.2>}
      localJointOffset=<5 0 0>
     }]
   }
   n9 = Anim Dynamics - Bone: hair_side_strandB_b_l(ComponentPose=@n8, Alpha=1, ExternalForce="0,0,0") {
    angularDampingOverride=1
    angularSpringConstant=1500
    bAngularSpring=true
    bGravityOverrideInSimSpace=true
    bOverrideAngularDamping=true
    gravityScale=0.35
    simulationSpace=RootRelative
    alphaBoolBlend {blendOption=Linear}
    boundBone {boneName=hair_side_strandB_b_l}
    componentLinearAccScale=<0.3 0.3 0.3>
    componentLinearVelScale=<0.3 0.3 0.3>
    physicsBodyDefinitions [{
      sphereCollisionRadius=0
      boundBone {boneName=hair_side_strandB_b_l}
      boxExtents=<10 15 15>
      constraintSetup {angularLimitsMax=<5 0 10>, angularLimitsMin=<-5 -20 -25>, angularTarget=<0 0 0>, linearAxesMax=<0.2 0.2 0.2>, linearAxesMin=<0.2 0.2 0.2>}
      localJointOffset=<5 0 0>
     }]
   }
   n10 = Anim Dynamics - Bone: hair_side_strand_r(ComponentPose=@n9, Alpha=0.7, ExternalForce="0,0,0") {
    angularDampingOverride=1
    angularSpringConstant=1500
    bAngularSpring=true
    bGravityOverrideInSimSpace=true
    bOverrideAngularDamping=true
    gravityScale=0.35
    simulationSpace=RootRelative
    alphaBoolBlend {blendOption=Linear}
    boundBone {boneName=hair_side_strand_r}
    componentLinearAccScale=<0.2 0.2 0.2>
    componentLinearVelScale=<0.2 0.2 0.2>
    physicsBodyDefinitions [{
      sphereCollisionRadius=0
      boundBone {boneName=hair_side_strand_r}
      boxExtents=<10 15 15>
      constraintSetup {angularLimitsMax=<5 0 20>, angularLimitsMin=<-5 -20 -30>, angularTarget=<0 0 0>}
      localJointOffset=<-5 0 0>
     }]
   }
   n11 = Anim Dynamics - Bone: hair_side_strandA_a_r(ComponentPose=@n10, Alpha=0.7, ExternalForce="0,0,0") {
    angularDampingOverride=1
    angularSpringConstant=1500
    bAngularSpring=true
    bGravityOverrideInSimSpace=true
    bOverrideAngularDamping=true
    gravityScale=0.35
    simulationSpace=RootRelative
    alphaBoolBlend {blendOption=Linear}
    boundBone {boneName=hair_side_strandA_a_r}
    componentLinearAccScale=<0.2 0.2 0.2>
    componentLinearVelScale=<0.2 0.2 0.2>
    physicsBodyDefinitions [{
      sphereCollisionRadius=0
      boundBone {boneName=hair_side_strandA_a_r}
      boxExtents=<10 15 15>
      constraintSetup {angularLimitsMax=<5 0 30>, angularLimitsMin=<-5 -20 -30>, angularTarget=<0 0 0>}
      localJointOffset=<-5 0 0>
     }]
   }
   n12 = Anim Dynamics - Bone: hair_side_strandB_a_r(ComponentPose=@n11, Alpha=0.7, ExternalForce="0,0,0") {
    angularDampingOverride=1
    angularSpringConstant=1500
    bAngularSpring=true
    bGravityOverrideInSimSpace=true
    bOverrideAngularDamping=true
    gravityScale=0.35
    simulationSpace=RootRelative
    alphaBoolBlend {blendOption=Linear}
    boundBone {boneName=hair_side_strandB_a_r}
    componentLinearAccScale=<0.2 0.2 0.2>
    componentLinearVelScale=<0.2 0.2 0.2>
    physicsBodyDefinitions [{
      sphereCollisionRadius=0
      boundBone {boneName=hair_side_strandB_a_r}
      boxExtents=<10 15 15>
      constraintSetup {angularLimitsMax=<5 0 5>, angularLimitsMin=<-5 -10 -5>, angularTarget=<0 0 0>}
      localJointOffset=<-5 0 0>
     }]
   }
   n13 = Anim Dynamics - Bone: hair_side_strandB_b_r(ComponentPose=@n12, Alpha=0.7, ExternalForce="0,0,0") {
    angularDampingOverride=1
    angularSpringConstant=1500
    bAngularSpring=true
    bGravityOverrideInSimSpace=true
    bOverrideAngularDamping=true
    gravityScale=0.35
    simulationSpace=RootRelative
    alphaBoolBlend {blendOption=Linear}
    boundBone {boneName=hair_side_strandB_b_r}
    componentLinearAccScale=<0.2 0.2 0.2>
    componentLinearVelScale=<0.2 0.2 0.2>
    physicsBodyDefinitions [{
      sphereCollisionRadius=0
      boundBone {boneName=hair_side_strandB_b_r}
      boxExtents=<10 15 15>
      constraintSetup {angularLimitsMax=<5 0 5>, angularLimitsMin=<-5 -10 -5>, angularTarget=<0 0 0>}
      localJointOffset=<-5 0 0>
     }]
   }
   n14 = Anim Dynamics - Bone: hair_bottomA_l(ComponentPose=@n13, Alpha=1, ExternalForce="0,0,0") {
    angularDampingOverride=1
    angularSpringConstant=1500
    bAngularSpring=true
    bGravityOverrideInSimSpace=true
    bOverrideAngularDamping=true
    gravityScale=0.35
    simulationSpace=RootRelative
    alphaBoolBlend {blendOption=Linear}
    boundBone {boneName=hair_bottomA_l}
    componentLinearAccScale=<0.3 0.3 0.3>
    componentLinearVelScale=<0.3 0.3 0.3>
    physicsBodyDefinitions [{
      sphereCollisionRadius=0
      boundBone {boneName=hair_bottomA_l}
      boxExtents=<20 10 10>
      constraintSetup {angularLimitsMax=<0 0 15>, angularLimitsMin=<0 -25 -15>, angularTarget=<0 0 0>}
      localJointOffset=<5 0 0>
     }]
   }
   n15 = Anim Dynamics - Bone: hair_bottomA_r(ComponentPose=@n14, Alpha=1, ExternalForce="0,0,0") {
    angularDampingOverride=1
    angularSpringConstant=1500
    bAngularSpring=true
    bGravityOverrideInSimSpace=true
    bOverrideAngularDamping=true
    gravityScale=0.35
    simulationSpace=RootRelative
    alphaBoolBlend {blendOption=Linear}
    boundBone {boneName=hair_bottomA_r}
    componentLinearAccScale=<0.3 0.3 0.3>
    componentLinearVelScale=<0.3 0.3 0.3>
    physicsBodyDefinitions [{
      sphereCollisionRadius=0
      boundBone {boneName=hair_bottomA_r}
      boxExtents=<20 10 10>
      constraintSetup {angularLimitsMax=<0 0 15>, angularLimitsMin=<0 -25 -15>, angularTarget=<0 0 0>}
      localJointOffset=<-5 0 0>
     }]
   }
   n16 = Control Rig(Source=Component To Local(ComponentPose=@n15), Alpha=Get isLOD0AndNOTFrontEnd() -> (isLOD0AndNOTFrontEnd:double)) {
    alphaBoolBlend {blendOption=Linear}
    controlRigAssetReference {blueprintRigClass=&/Glasses_ControlRig.Glasses_ControlRig_C}
   }
   Output Pose(Result=Component To Local(ComponentPose=Local To Component(LocalPose=@n16))) {layerGroup=None}
  ]}
 EventGraph {nodes [
   Event BlueprintUpdateAnimation()
   > n1 = Sequence!(execute) -> (then_0, then_1, then_2:Exec)
   n2 = Branch(execute=@n1.then_1, Condition=Get bIsInFrontEnd())
   > Set isFrontEnd(isFrontEnd:double=0) -> (Output_Get:double)
   Set isFrontEnd(execute=@n2.else, isFrontEnd=1)
   Set isLOD0AndNOTFrontEnd(execute=@n1.then_2, isLOD0AndNOTFrontEnd:double=float * float(A=Get isLOD0() -> (isLOD0:double), B=Get isFrontEnd() -> (isFrontEnd:double))) -> (Output_Get:double)
   n3 = Branch(execute=@n1.then_0, Condition="Equal (Integer)"(A=Get CurrentLOD(), B=0) {is=title K2Node_CallFunction})
   > Set isLOD0(isLOD0:double=1) -> (Output_Get:double)
   Set isLOD0(execute=@n3.else, isLOD0=0)
   Event BlueprintInitializeAnimation()
   > Set Body_Mesh(Body_Mesh=GetOwnerSkeletalMeshComponent())
   > n4 = Cast To PlayerPawn_Athena(Object=TryGetPawnOwner())
   > n5 = Set Pawn Owner(Pawn Owner:Player Pawn Athena Object Reference=@n4) -> (Output_Get:Player Pawn Athena Object Reference)
   > Set Head_Mesh(Head_Mesh:Skeletal Mesh Component Object Reference=GetSkeletalMeshForPartType(self=@n5, PartType=Head)) -> (Output_Get:Skeletal Mesh Component Object Reference)
  ]}
]
variables [
 Body_Mesh {type=SkeletalMeshComponent}
 Head_Mesh {type=SkeletalMeshComponent}
 isFrontEnd {type=number}
 isLOD0 {type=number}
 isLOD0AndNOTFrontEnd {type=number}
 Pawn Owner {type=PlayerPawn_Athena_C}
]
